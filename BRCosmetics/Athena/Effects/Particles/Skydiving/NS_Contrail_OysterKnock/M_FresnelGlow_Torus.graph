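# Unlit additive Niagara mesh-particle material that gives a Fresnel glow for the OysterKnock skydiving contrail, previewed on a rings mesh.
# Particle color times double-sided Fresnel feeds a shared emissive adjustment function.
# Opacity combines Fresnel, inverted vertex red, particle alpha and blurred noise panned vertically by a RandomSpeed dynamic parameter.
# A quality switch applies depth fade to that opacity, skipping it on Low and Medium.
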Material /BRCosmetics/Athena/Effects/Particles/Skydiving/NS_Contrail_OysterKnock/M_FresnelGlow_Torus
graph {
 properties {
  bAutomaticallySetUsageInEditor=false
  bUsedWithNiagaraMeshParticles=true
  blendMode=BLEND_Additive
  pixelDepthOffsetMode=PDOM_Legacy
  previewMesh=&/SM_OysterKnock_Rings
  shadingModel=MSM_Unlit
  usedShadingModels=Unlit
 }
 nodes [
  n1 = ParticleColor()
  n2 = MaterialFunctionCall(Brightness=ScalarParameter() {defaultValue=1, parameterName=Brightness}, Exponent=ScalarParameter() {defaultValue=1, parameterName=Exponent}) {materialFunction=&MF_Fresnel_FX}
  MaterialFunctionCall(Emissive=Multiply(A=@n1.RGB, B=@n2.Fresnel Double Sided)) {materialFunction=&MF_FX_EmissiveAdjustments}
  n3 = OneMinus(None=VertexColor().R)
  n4 = DynamicParameter() {paramNames [RandomSpeed,Param2,Param3,Param4]}
  n5 = Constant2Vector() {g=1.5}
  n6 = TextureSample(UVs=Panner(Coordinate=TextureCoordinate() {vTiling=0.5}, Speed=Multiply(A=@n4.RandomSpeed, B=@n5)) {speedX=0.1, speedY=-0.1}) {samplerType=SAMPLERTYPE_Grayscale, texture=&LowResBlurredNoise}
  n7 = Multiply(A=@n1.A, B=Multiply(A=Multiply(A=@n2.Fresnel, B=@n3), B=@n6.RGB))
  QualitySwitch(Default=DepthFade(Opacity=@n7, FadeDistance=ScalarParameter() {defaultValue=2, parameterName=DepthFade}), Low=@n7, Medium=@n7)
  Saturate(None=@n6.RGB)
 ]
}
